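# Translucent unlit Niagara sprite material for a full-screen scene-tint transition, attached to the camera through MF_DLR_FullScreenTransition_Alpha.
# It desaturates and tints scene color, adding world-normal fresnel, Beer's-law height fog, a noise-broken transition edge and vignette.
# Particle color, dynamic parameters and a custom-stencil mask control the look.
# Platform switches replace scene color, fresnel and stencil with constants on PC-Perf, Android and Switch.
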Material /DealershipSharedLibrary/FX/Materials/MasterMaterials/FullScreenTransition/M_DLR_FullScreen_SceneTint
graph {
 properties {
  bAutomaticallySetUsageInEditor=false
  bUseTranslucencyVertexFog=false
  bUsedWithNiagaraSprites=true
  bUsedWithStaticMesh=false
  blendMode=BLEND_Translucent
  maxWorldPositionOffsetDisplacement=10000000
  previewMesh=/Engine/Transient.MaterialEditorMeshComponent_10
  shadingModel=MSM_Unlit
  usedShadingModels=Unlit
 }
 nodes [
  n1 = NamedRerouteUsage()
  n2 = DynamicParameter() {defaultValue=<rgba 1 2 1.1236 0>, paramNames [Scene - Desaturation, Wrold Fresnel - Exp, Stencil Mask - Opacity, Is Painted]}
  MaterialFunctionCall(Emissive=Multiply(A=ParticleColor().RGB, B=Desaturation(None=@n1, Fraction=@n2.Is Painted))) {materialFunction=&MF_FX_EmissiveAdjustments}
  Multiply(A=ParticleColor().A, B=Multiply(A=NamedRerouteUsage(), B=NamedRerouteUsage()))
  n3 = MaterialFunctionCall!(Use MPC, Is Sprite, Use World Position Behind Translucency=StaticBool() {value=true}) {materialFunction=/DealershipSharedLibrary/FX/Materials/MaterialFunctions/MF_DLR_FullScreenTransition_Alpha}
  NamedRerouteDeclaration(None=@n3.Attach to camera WPO) {name=WPO, nodeColor=<rgba 0 1 0.3203 1>}
  NamedRerouteDeclaration(None=@n3.Curve - Alpha) {name=Alpha, nodeColor=<rgba 0.1325 0.4896 0.3189 1>}
  n4 = Constant() {r=1}
  n5 = DynamicParameter() {defaultValue=<rgba 1 2 1.1236 0>, paramNames [Scene - Desaturation, Wrold Fresnel - Exp, Stencil Mask - Opacity, Is Painted]}
  n6 = Subtract(
   A=ComponentMask(None=SceneTexture() {sceneTextureId=PPI_CustomDepth}.Color) {r=true}
   B=ComponentMask(None=SceneTexture() {sceneTextureId=PPI_SceneDepth}.Color) {r=true}
  )
  n7 = If(
   A=ComponentMask(None=SceneTexture() {sceneTextureId=PPI_CustomStencil}.Color) {r=true}
   B=ScalarParameter() {defaultValue=1, group=Stencil Mask, parameterName=Stencil Mask - Stencil Value}
   "A > B"=@n4
   "A == B"=Max(A=@n5.Stencil Mask - Opacity, B=Saturate(None=@n6))
   "A < B"=@n4
  )
  n8 = Constant() {r=1}
  NamedRerouteDeclaration(None=MaterialFunctionCall("Xb1,PS4"=@n7, "PC-High&&Epic,XSX,PS5"=@n7, "PC-Perf,Android"=@n8, Switch=@n8, "PC-Low(NotPerf)&&Medium"=@n7) {materialFunction=&PlatformSwitch_1D}.Result) {name=Stencil Mask, nodeColor=<rgba 1 0 0.0703 1>}
  n9 = DynamicParameter() {defaultValue=<rgba 1 2 1.1236 0>, paramNames [Scene - Desaturation, Wrold Fresnel - Exp, Stencil Mask - Opacity, Is Painted]}
  n10 = Fresnel(ExponentIn=@n9.Wrold Fresnel - Exp, Normal=ComponentMask(None=SceneTexture() {sceneTextureId=PPI_WorldNormal}.Color) {g=true, r=true}) {baseReflectFraction=0}
  n11 = Saturate(None=LinearInterpolate(A=@n10, B=OneMinus(None=@n10), Alpha=ScalarParameter() {group=World Fresnel, parameterName=WorldFresnel - Invert}))
  n12 = Constant()
  NamedRerouteDeclaration(None=MaterialFunctionCall("Xb1,PS4"=@n11, "PC-High&&Epic,XSX,PS5"=@n11, "PC-Perf,Android"=@n12, Switch=@n12, "PC-Low(NotPerf)&&Medium"=@n11) {materialFunction=&PlatformSwitch_1D}.Result) {name=World Fresnel Mask, nodeColor=<rgba 0.4297 0 1 1>}
  n13 = ComponentMask(None=@n3.Curve - RGB) {r=true}
  n14 = ScalarParameter() {group=Edge Noise, parameterName=Use Edge Noise Texture}
  n15 = TextureObjectParameter() {group=Edge Noise, parameterName=Edge Noise Texture, sortPriority=2, texture=/DealershipSharedLibrary/FX/Textures/Noise/T_VCVFX_BlobNoise}
  n16 = Power(
   Base=MaterialFunctionCall(
    TextureObject=@n15
    TextureSize=ScalarParameter() {defaultValue=64, group=Edge Noise, parameterName=Edge Noise Texture - Tiling}
    ProjectionTransitionContrast=ScalarParameter() {defaultValue=1, group=Edge Noise, parameterName=Edge Noise Texture - TransitionContrast}
   ) {materialFunction=&WorldAlignedTexture}.XY Texture
   Exp=ScalarParameter() {defaultValue=1, group=Edge Noise, parameterName=Edge Noise Texture - Power}
  )
  n17 = DynamicParameter() {parameterIndex=2, defaultValue=<rgba 1000 0 1 1>, paramNames [Height Fog - Falloff, Edge Noise Texture - Amount, Param3, Param4]}
  n18 = LinearInterpolate(Alpha=@n17.Edge Noise Texture - Amount) {constA=1, constB=0}
  n19 = ScalarParameter() {defaultValue=64, group=Edge Noise, parameterName=Edge Noise Overlay Texture - Tiling}
  n20 = Divide(A=WorldPosition() {worldPositionShaderOffset=WPT_ExcludeAllShaderOffsets}.XYZ, B=Multiply(A=Abs(None=@n19)) {constB=-1})
  n21 = ScalarParameter() {defaultValue=0.6, parameterName=NoiseOverlaySpeed}
  n22 = Constant()
  n23 = TextureSample(UVs=Panner(Coordinate=ComponentMask(None=@n20) {b=true, g=true}, Speed=AppendVector(A=@n21, B=@n22))) {texture=/Game/Effects/Fort_Effects/Textures/Tile/T_Noise_03}
  n24 = Power(
   Base=Multiply(A=Desaturation(None=@n23.RGB), B=ScalarParameter() {defaultValue=20, parameterName=EdgeNoiseOverlay-Mult})
   Exp=ScalarParameter() {defaultValue=2.0759, parameterName=EdgeNoiseOverlay-Pwr}
  )
  n25 = Add(A=Saturate(None=Add(A=@n16, B=@n18)), B=Clamp(None=@n24, Max=@n24))
  NamedRerouteDeclaration(None=Multiply(A=@n13, B=If(A=@n14, "A > B"=@n25, "A == B"=@n25, "A < B"=@n13) {constB=1})) {name=Edge Mask, nodeColor=<rgba 1 0 0.6094 1>}
  NamedRerouteDeclaration(None=@n3.Edge Color) {name=Edge Color, nodeColor=<rgba 1 0.3516 0 1>}
  n26 = ScalarParameter() {group=Vignette, parameterName=Vignette - Edge}
  n27 = ScalarParameter() {group=Vignette, parameterName=Vignette - Center}
  n28 = Power(
   Base=SphereMask(
    A=TextureCoordinate()
    B=Constant() {r=0.5}
    Radius=ScalarParameter() {defaultValue=0.71, group=Vignette, parameterName=Vignette - Radius}
    Hardness=ScalarParameter() {group=Vignette, parameterName=Vignette - Hardness}
   ) {attenuationRadius=0.71}
   Exp=ScalarParameter() {defaultValue=1, group=Vignette, parameterName=Vignette- Exp}
  ) {constExponent=1}
  NamedRerouteDeclaration(None=Saturate(None=LinearInterpolate(A=@n26, B=@n27, Alpha=@n28))) {name=Vignette, nodeColor=<rgba 1 0.0183 0.597 1>}
  n29 = SceneColor!(UVs)
  n30 = Constant() {r=1}
  n31 = DynamicParameter() {defaultValue=<rgba 1 2 1.1236 0>, paramNames [Scene - Desaturation, Wrold Fresnel - Exp, Stencil Mask - Opacity, Is Painted]}
  n32 = Desaturation(
   None=MaterialFunctionCall("Xb1,PS4"=@n29.RGB, "PC-High&&Epic,XSX,PS5"=@n29.RGB, "PC-Perf,Android"=@n30, Switch=@n30, "PC-Low(NotPerf)&&Medium"=@n29.RGB) {materialFunction=&PlatformSwitch_1D}.Result
   Fraction=@n31.Scene - Desaturation
  )
  n33 = VectorParameter() {group=Scene, parameterName=Scene - Color Tint, defaultValue=<rgba 1 1 1 1>}
  n34 = VectorParameter() {group=World Fresnel, parameterName=World Fresnel - Color, defaultValue=<rgba 0.1823 0.1823 0.1823 1>}
  n35 = NamedRerouteUsage()
  n36 = LinearInterpolate(
   A=VectorParameter() {group=Fog, parameterName=Fog - Color, defaultValue=<rgba 0.349 0.349 0.349 1>}.RGB
   B=Add(A=Multiply(A=@n32, B=@n33.RGB), B=Multiply(A=@n34.RGB, B=@n35))
   Alpha=NamedRerouteUsage()
  )
  n37 = ComponentMask(None=NamedRerouteUsage()) {b=true, g=true, r=true}
  n38 = NamedRerouteUsage()
  NamedRerouteDeclaration(None=LinearInterpolate(
    A=LinearInterpolate(A=@n36, B=@n37, Alpha=@n38)
    B=VectorParameter() {group=Vignette, parameterName=Vignette - Color, defaultValue=<rgba 0 0 0 1>}.RGB
    Alpha=NamedRerouteUsage()
   )) {name=Final Scene, nodeColor=<rgba 0.6696 0.2708 1 1>}
  n39 = DynamicParameter() {parameterIndex=1, defaultValue=<rgba 100 0.01 0.5 1>, paramNames [Fog - Start Distance, Fog - Density, Fog - Opacity, Height Fog - Offset]}
  n40 = Multiply(A=@n39.Fog - Density) {constB=0.001}
  n41 = OneMinus(None=ComponentMask(None=MaterialFunctionCall() {materialFunction=&WorldPositionBehindTranslucency}.Result) {b=true})
  n42 = Multiply(A=@n39.Height Fog - Offset) {constB=500}
  n43 = Saturate(None=Multiply(A=Add(A=@n41, B=@n42), B=@n40))
  n44 = DynamicParameter() {parameterIndex=2, defaultValue=<rgba 1000 0 1 1>, paramNames [Height Fog - Falloff, Edge Noise Texture - Amount, Param3, Param4]}
  n45 = Multiply(A=@n44.Height Fog - Falloff) {constB=7}
  n46 = MaterialFunctionCall!(Thickness=Add(A=@n40, B=Power(Base=@n43, Exp=@n45)), Depth Scale=Subtract(A=SceneDepth().R, B=@n39.Fog - Start Distance)) {materialFunction=/Engine/Functions/Engine_MaterialFunctions01/Volumetrics/BeersLaw}
  NamedRerouteDeclaration(None=LinearInterpolate(A=Saturate(None=@n46.Result), Alpha=OneMinus(None=@n39.Fog - Opacity))) {name=Fog Mask, nodeColor=<rgba 0.3828 0 1 1>}
  NamedRerouteUsage()
  NamedRerouteUsage()
  n47 = SphereMask(
   A=TextureCoordinate()
   B=Constant2Vector() {g=0.5, r=0.5}
   Radius=ScalarParameter() {defaultValue=0.94, group=Mask, parameterName=Fresnel - SphereMask_Radius}
   Hardness=ScalarParameter() {group=Mask, parameterName=Fresnel - SphereMask_Hardness}
  )
  n48 = ScalarParameter() {defaultValue=7.6021, parameterName=Fresnel - SphereMask Multiplier}
  Saturate(None=Power(Base=Multiply(A=@n47, B=@n48)))
 ]
}
