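# Translucent unlit Niagara-sprite master material for full-screen transitions that restyle the scene with tint, fog and geometry outlines.
# SceneColor gets desaturation, tint and normal-based fresnel, blended with Beer's-law height fog and optional panning texture.
# Outlines use neighbouring depth samples, with optional dithered light-vector shadows and dot hatching, plus normal-derived detail lines.
# A transition function supplies camera-attached WPO, alpha and noise-breakable edge; custom stencil masks cars.
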Material /DealershipSharedLibrary/FX/Materials/MasterMaterials/FullScreenTransition/M_DLR_FullScreen_SceneTint_GeoOutline
graph {
 properties {
  bAutomaticallySetUsageInEditor=false
  bUseTranslucencyVertexFog=false
  bUsedWithNiagaraSprites=true
  blendMode=BLEND_Translucent
  maxWorldPositionOffsetDisplacement=10000000
  shadingModel=MSM_Unlit
  usedShadingModels=Unlit
 }
 nodes [
  n1 = NamedRerouteUsage()
  n2 = DynamicParameter() {defaultValue=<rgba 1 2 1.1236 0>, paramNames [Scene - Desaturation, Wrold Fresnel - Exp, Stencil Mask - Opacity, Is Painted]}
  MaterialFunctionCall(Emissive=Multiply(A=ParticleColor().RGB, B=Desaturation(None=@n1, Fraction=@n2.Is Painted))) {materialFunction=&MF_FX_EmissiveAdjustments}
  n3 = Constant() {r=1}
  n4 = DynamicParameter() {defaultValue=<rgba 1 2 1.1236 0>, paramNames [Scene - Desaturation, Wrold Fresnel - Exp, Stencil Mask - Opacity, Is Painted]}
  n5 = Subtract(
   A=ComponentMask(None=SceneTexture() {sceneTextureId=PPI_CustomDepth}.Color) {r=true}
   B=ComponentMask(None=SceneTexture() {sceneTextureId=PPI_SceneDepth}.Color) {r=true}
  )
  n6 = If(
   A=ComponentMask(None=SceneTexture() {sceneTextureId=PPI_CustomStencil}.Color) {r=true}
   B=ScalarParameter() {defaultValue=1, group=Stencil Mask, parameterName=Stencil Mask - Stencil Value}
   "A > B"=@n3
   "A == B"=Max(A=@n4.Stencil Mask - Opacity, B=Saturate(None=@n5))
   "A < B"=@n3
  )
  Multiply(
   A=ParticleColor().A
   B=Multiply(A=NamedRerouteUsage(), B=NamedRerouteDeclaration(None=@n6) {name=Stencil Mask, nodeColor=<rgba 1 0 0.0703 1>})
  )
  n7 = MaterialFunctionCall!(Use MPC, Is Sprite=ScalarParameter() {parameterName=Is Sprite?}, Use World Position Behind Translucency=StaticBool() {value=true}) {materialFunction=/DealershipSharedLibrary/FX/Materials/MaterialFunctions/MF_DLR_FullScreenTransition_Alpha}
  NamedRerouteDeclaration(None=@n7.Attach to camera WPO) {name=WPO, nodeColor=<rgba 0 1 0.3203 1>}
  NamedRerouteDeclaration(None=@n7.Curve - Alpha) {name=Alpha, nodeColor=<rgba 0.1325 0.4896 0.3189 1>}
  n8 = DynamicParameter() {defaultValue=<rgba 1 2 1.1236 0>, paramNames [Scene - Desaturation, Wrold Fresnel - Exp, Stencil Mask - Opacity, Is Painted]}
  n9 = Fresnel(ExponentIn=@n8.Wrold Fresnel - Exp, Normal=ComponentMask(None=SceneTexture() {sceneTextureId=PPI_WorldNormal}.Color) {g=true, r=true}) {baseReflectFraction=0}
  n10 = OneMinus(None=@n9)
  n11 = ScalarParameter() {group=World Fresnel, parameterName=WorldFresnel - Invert}
  NamedRerouteDeclaration(None=Saturate(None=LinearInterpolate(A=@n9, B=@n10, Alpha=@n11))) {name=World Fresnel Mask, nodeColor=<rgba 0.4297 0 1 1>}
  n12 = TextureObjectParameter() {group=Edge Noise, parameterName=Edge Noise Texture, sortPriority=2, texture=/DealershipSharedLibrary/FX/Textures/Noise/T_VCVFX_BlobNoise}
  n13 = MaterialFunctionCall(TextureObject=@n12, TextureSize=ScalarParameter() {defaultValue=64, group=Edge Noise, parameterName=Edge Noise Texture - Tiling}) {materialFunction=&WorldAlignedTexture}
  n14 = ScalarParameter() {defaultValue=1, group=Edge Noise, parameterName=Edge Noise Texture - Power}
  n15 = DynamicParameter() {parameterIndex=2, defaultValue=<rgba 1000 0 1 1>, paramNames [Height Fog - Falloff, Edge Noise Texture - Amount, Param3, Param4]}
  n16 = Clamp(None=Add(A=Power(Base=@n13.XY Texture, Exp=@n14), B=LinearInterpolate(Alpha=@n15.Edge Noise Texture - Amount) {constA=1, constB=0}))
  n17 = Constant() {r=1}
  NamedRerouteDeclaration(None=Multiply(
    A=ComponentMask(None=@n7.Curve - RGB) {r=true}
    B=StaticSwitchParameter(True=@n16, False=@n17) {group=Edge Noise, parameterName=Use Edge Noise Texture?, sortPriority=1}
   )) {name=Edge Mask, nodeColor=<rgba 1 0 0.6094 1>}
  n18 = SceneDepth()
  n19 = ScalarParameter() {defaultValue=0, group=Outline, parameterName=Outline Depth, sortPriority=16}
  n20 = Saturate(None=Divide(A=Subtract(A=@n18.R, B=@n19), B=ScalarParameter() {group=Outline, parameterName=Outline Depth Offset, sortPriority=16}))
  NamedRerouteDeclaration(None=@n20) {name=Outline Depth Com, nodeColor=<rgba 0.4531 0 1 1>}
  n21 = ScreenPosition()
  n22 = ScalarParameter() {defaultValue=1, group=Outline, parameterName=Outline Thickness, sortPriority=2}
  n23 = Multiply(A=ScalarParameter() {defaultValue=1, group=Outline, parameterName=Near, sortPriority=4}, B=@n22)
  n24 = Multiply(A=ScalarParameter() {defaultValue=1, group=Outline, parameterName=Far, sortPriority=8}, B=@n22)
  n25 = MaterialFunctionCall!(OffsetAmount=Multiply(A=SceneTexelSize(), B=LinearInterpolate(A=@n23, B=@n24, Alpha=@n20))) {materialFunction=/Engine/Functions/Engine_MaterialFunctions01/ImageAdjustment/BlurSampleOffsets}
  n26 = Add(A=SceneDepth(UVs=Add(A=@n21.ViewportUV, B=@n25."0,1")).R, B=SceneDepth(UVs=Add(A=@n21.ViewportUV, B=@n25."1,0")).R)
  n27 = Add(A=SceneDepth(UVs=Add(A=@n21.ViewportUV, B=@n25."0,-1")).R, B=SceneDepth(UVs=Add(A=@n21.ViewportUV, B=@n25."-1,0")).R)
  n28 = Subtract(A=Multiply(A=Add(A=@n26, B=@n27)) {constB=0.25}, B=SceneDepth().R)
  n29 = LinearInterpolate(
   A=ScalarParameter() {defaultValue=64, group=Outline, parameterName=Outline Div, sortPriority=20}
   B=ScalarParameter() {defaultValue=0.4097, group=Outline, parameterName=Car - Outline Div, sortPriority=21}
   Alpha=NamedRerouteUsage()
  )
  n30 = ScalarParameter() {defaultValue=64, group=Outline, parameterName=Outline Pow}
  n31 = ScalarParameter() {defaultValue=52.1747, group=Outline, parameterName=Car - Outline Pow, sortPriority=33}
  n32 = NamedRerouteUsage()
  n33 = Saturate(None=Power(Base=Divide(A=@n28, B=@n29), Exp=LinearInterpolate(A=@n30, B=@n31, Alpha=@n32)))
  n34 = ScalarParameter() {defaultValue=0.4, group=Outline Shadow, parameterName=Shadow Close, sortPriority=16}
  n35 = ScalarParameter() {defaultValue=1.2, group=Outline Shadow, parameterName=Shadow Far, sortPriority=20}
  n36 = Subtract(A=SceneDepth().R, B=ScalarParameter() {defaultValue=5000, group=Outline Shadow, parameterName=Shadow Bias, sortPriority=8})
  n37 = ScalarParameter() {defaultValue=100000, group=Outline Shadow, parameterName=Shadow Distance, sortPriority=10}
  n38 = Add(
   A=Saturate(None=Divide(A=@n36, B=@n37))
   B=Multiply(
    A=MaterialFunctionCall() {materialFunction=&DitherTemporalAA}.Result
    B=ScalarParameter() {defaultValue=0.1, group=Outline Shadow, parameterName=Shadow Noise Amount, sortPriority=12}
   ) {constB=0.1}
  )
  n39 = DotProduct(
   A=SceneTexture() {sceneTextureId=PPI_WorldNormal}.Color
   B=VectorParameter() {group=Outline Shadow, parameterName=LightVector, sortPriority=4, defaultValue=<rgba 0 0 -1 1>}.RGB
  )
  n40 = Max(A=@n33, B=Step(Y=LinearInterpolate(A=@n34, B=@n35, Alpha=@n38) {constA=1}, X=ConstantBiasScale(None=@n39)))
  NamedRerouteDeclaration(None=If(A=ScalarParameter() {group=Outline Shadow, parameterName=Use Outline Shadows, sortPriority=0}, "A > B"=@n40, "A == B"=@n40, "A < B"=@n33) {constB=1}) {name=Outline Connections, nodeColor=<rgba 0.3005 0.3918 1 1>}
  n41 = SceneTexture() {sceneTextureId=PPI_CustomStencil}
  n42 = Constant()
  NamedRerouteDeclaration(None=If(
    A=ComponentMask(None=@n41.Color) {r=true}
    B=ScalarParameter() {defaultValue=1, group=Stencil Mask, parameterName=Stencil Mask - Stencil Value}
    "A > B"=@n42
    "A == B"=Constant() {r=1}
    "A < B"=@n42
   )) {name=Car Mask, nodeColor=<rgba 0 1 0.1797 1>}
  NamedRerouteDeclaration(None=@n7.Edge Color) {name=Edge Color, nodeColor=<rgba 1 0.3516 0 1>}
  n43 = Multiply(A=NamedRerouteUsage(), B=NamedRerouteUsage())
  NamedRerouteDeclaration(None=If(
    A=ScalarParameter() {group=Hatching, parameterName=Use Outline Hatching, sortPriority=0}
    "A > B"=@n43
    "A == B"=@n43
    "A < B"=NamedRerouteUsage()
   ) {constB=1}) {name=Outline, nodeColor=<rgba 0 0.7031 0.6873 1>}
  n44 = ComponentMask(None=ScreenPosition().ViewportUV) {g=true, r=true}
  n45 = Constant2Vector() {g=0.5, r=1}
  n46 = TextureSampleParameter2D(UVs=Divide(A=Multiply(A=@n44, B=@n45), B=ScalarParameter() {defaultValue=0.01, group=Hatching, parameterName=HatchTile, sortPriority=6})) {
   group=Hatching
   parameterName=Hatching Texture
   samplerType=SAMPLERTYPE_Grayscale
   sortPriority=4
   texture=/DealershipSharedLibrary/FX/Textures/Masks/T_GradientCircle
  }
  n47 = ScalarParameter() {defaultValue=0.14, group=Hatching, parameterName=Hatch Thickness Min, sortPriority=14}
  n48 = ScalarParameter() {defaultValue=-0.35, group=Hatching, parameterName=Hatch Thickness Max, sortPriority=12}
  n49 = SceneTexture() {sceneTextureId=PPI_WorldNormal}
  n50 = Constant() {r=1}
  n51 = Multiply(A=@n49.Color, B=@n50)
  n52 = Multiply(A=Ceil(None=Divide(A=@n51, B=@n50)), B=@n49.Color)
  n53 = Constant() {r=0.2}
  n54 = Clamp(None=Desaturation(None=Divide(A=@n52, B=@n53), Fraction=Constant())) {minDefault=0.01}
  n55 = Saturate(None=Subtract(A=Desaturation(None=@n54), B=ScalarParameter() {defaultValue=0.683, desc=Softness, group=Hatching, parameterName=Crunch, sortPriority=16}))
  n56 = Divide(
   A=Subtract(A=SceneDepth().R, B=ScalarParameter() {defaultValue=0, group=Hatching, parameterName=Hatch Depth, sortPriority=9})
   B=ScalarParameter() {defaultValue=0.34, group=Hatching, parameterName=HatchDepthPower, sortPriority=10}
  )
  n57 = Subtract(A=@n46.R, B=LinearInterpolate(A=Constant() {r=-1}, B=LinearInterpolate(A=@n47, B=@n48, Alpha=@n55), Alpha=Saturate(None=@n56)))
  n58 = Constant() {r=8}
  NamedRerouteDeclaration(None=Saturate(None=Multiply(A=@n57, B=@n58))) {name=Dot Hatch, nodeColor=<rgba 1 0.4167 0.7845 1>}
  n59 = AppendVector(
   A=ScalarParameter() {group=Fog Texture, parameterName=Fog Tex - Offset X, sortPriority=8}
   B=ScalarParameter() {group=Fog Texture, parameterName=Fog Tex - Offset Y, sortPriority=9}
  )
  n60 = VectorParameter() {
   group=Fog Texture
   parameterName=Fog Tex - UVs
   sortPriority=4
   channelNames {a=Pan Y, b=Pan X, g=Scale Y, r=Scale X}
   defaultValue=<rgba 1 1 0 0>
  }
  n61 = Multiply(A=TextureCoordinate(), B=ComponentMask(None=@n60.RGB) {g=true, r=true})
  n62 = AppendVector(A=@n60.Pan X, B=@n60.Pan X)
  n63 = Time()
  n64 = TextureSampleParameter2D!(UVs=Add(A=Add(A=@n59, B=@n61), B=Multiply(A=@n62, B=@n63)), Level, Apply View MipBias) {
   automaticViewMipBias=false
   group=Fog Texture
   mipValueMode=TMVM_MipLevel
   parameterName=Fog - Texture
   sortPriority=2
   texture=/DealershipSharedLibrary/Assets/Sets/Gravity/Textures/T_FX_Gravity_Nebula_RGB
  }
  NamedRerouteDeclaration(None=@n64.RGB) {name=Fog Texture, nodeColor=<rgba 1 0.3757 0.9801 1>}
  n65 = ScalarParameter() {group=Outline, parameterName=Use Outline, sortPriority=0}
  n66 = NamedRerouteUsage()
  n67 = VectorParameter() {group=Fog, parameterName=Fog - Color, defaultValue=<rgba 0.349 0.349 0.349 1>}
  n68 = LinearInterpolate(
   A=@n66
   B=Multiply(A=@n67.RGB, B=@n66)
   Alpha=ScalarParameter() {defaultValue=1, group=Fog Texture, parameterName=Fog - Texture Blend with Fog Color, sortPriority=3}
  )
  n69 = Constant() {r=1}
  n70 = DynamicParameter() {defaultValue=<rgba 1 2 1.1236 0>, paramNames [Scene - Desaturation, Wrold Fresnel - Exp, Stencil Mask - Opacity, Is Painted]}
  n71 = Desaturation(None=FeatureLevelSwitch(Default=SceneColor!(UVs).RGB, ES2=@n69, ES3_1=@n69, SM4_REMOVED=@n69), Fraction=@n70.Scene - Desaturation)
  n72 = VectorParameter() {group=Scene, parameterName=Scene - Color Tint, defaultValue=<rgba 1 1 1 1>}
  n73 = VectorParameter() {group=World Fresnel, parameterName=World Fresnel - Color, defaultValue=<rgba 0.1823 0.1823 0.1823 1>}
  n74 = NamedRerouteUsage()
  n75 = LinearInterpolate(
   A=If(A=ScalarParameter() {group=Fog Texture, parameterName=Use Fog Texture, sortPriority=0}, "A > B"=@n68, "A == B"=@n68, "A < B"=@n67.RGB) {constB=1}
   B=Add(A=Multiply(A=@n71, B=@n72.RGB), B=Multiply(A=@n73.RGB, B=@n74))
   Alpha=NamedRerouteUsage()
  )
  n76 = LinearInterpolate(
   A=@n75
   B=VectorParameter() {group=Outline, parameterName=Outline Color, sortPriority=1, defaultValue=<rgba 0 1 0.9699 1>}.RGB
   Alpha=Saturate(None=NamedRerouteUsage())
  )
  n77 = NamedRerouteUsage()
  NamedRerouteDeclaration(None=LinearInterpolate(
    A=If(A=@n65, "A > B"=@n76, "A == B"=@n76, "A < B"=@n75) {constB=1}
    B=ComponentMask(None=@n77) {b=true, g=true, r=true}
    Alpha=NamedRerouteUsage()
   )) {name=Final Scene, nodeColor=<rgba 0.6696 0.2708 1 1>}
  n78 = Ceil(None=NamedRerouteUsage())
  n79 = Constant2Vector() {r=1}
  n80 = ScalarParameter() {defaultValue=0.005, group=Detail Lines, parameterName=HeightBias, sortPriority=4}
  n81 = Constant()
  n82 = ComponentMask(None=ScreenPosition().ViewportUV) {g=true, r=true}
  n83 = SceneTexture(UVs=Add(A=AppendVector(A=@n80, B=@n81), B=@n82)) {sceneTextureId=PPI_WorldNormal}
  n84 = ComponentMask(None=SceneTexture(UVs=@n82) {sceneTextureId=PPI_WorldNormal}.Color) {r=true}
  n85 = LinearInterpolate(
   A=ScalarParameter() {defaultValue=8, group=Detail Lines, parameterName=NormalHeight, sortPriority=12}
   B=Constant() {r=0.3}
   Alpha=NamedRerouteUsage()
  )
  n86 = Multiply(A=Subtract(A=ComponentMask(None=@n83.Color) {r=true}, B=@n84), B=@n85)
  n87 = Constant2Vector() {g=1}
  n88 = SceneTexture(UVs=Add(A=@n82, B=AppendVector(A=@n81, B=@n80))) {sceneTextureId=PPI_WorldNormal}
  n89 = Multiply(A=@n85, B=Subtract(A=ComponentMask(None=@n88.Color) {r=true}, B=@n84))
  n90 = ConstantBiasScale(None=CrossProduct(A=AppendVector(A=@n79, B=@n86), B=AppendVector(A=@n87, B=@n89))) {bias=0.5}
  n91 = Subtract(A=OneMinus(None=Desaturation(None=@n90)), B=ScalarParameter() {group=Detail Lines, parameterName=Height Power, sortPriority=4})
  NamedRerouteDeclaration(None=Multiply(A=Saturate(None=@n78), B=Ceil(None=@n91))) {name=Detail Lines, nodeColor=<rgba 1 0.3516 0 1>}
  n92 = DynamicParameter() {parameterIndex=1, defaultValue=<rgba 100 0.01 0.5 1>, paramNames [Fog - Start Distance, Fog - Density, Fog - Opacity, Height Fog - Offset]}
  n93 = Multiply(A=@n92.Fog - Density) {constB=0.001}
  n94 = OneMinus(None=ComponentMask(None=MaterialFunctionCall() {materialFunction=&WorldPositionBehindTranslucency}.Result) {b=true})
  n95 = Multiply(A=@n92.Height Fog - Offset) {constB=500}
  n96 = Saturate(None=Multiply(A=Add(A=@n94, B=@n95), B=@n93))
  n97 = DynamicParameter() {parameterIndex=2, defaultValue=<rgba 1000 0 1 1>, paramNames [Height Fog - Falloff, Edge Noise Texture - Amount, Param3, Param4]}
  n98 = Multiply(A=@n97.Height Fog - Falloff) {constB=7}
  n99 = MaterialFunctionCall!(Thickness=Add(A=@n93, B=Power(Base=@n96, Exp=@n98)), Depth Scale=Subtract(A=SceneDepth().R, B=@n92.Fog - Start Distance)) {materialFunction=/Engine/Functions/Engine_MaterialFunctions01/Volumetrics/BeersLaw}
  NamedRerouteDeclaration(None=LinearInterpolate(A=Saturate(None=@n99.Result), Alpha=OneMinus(None=@n92.Fog - Opacity))) {name=Fog Mask, nodeColor=<rgba 0.3828 0 1 1>}
  NamedRerouteUsage()
  NamedRerouteUsage()
 ]
}
